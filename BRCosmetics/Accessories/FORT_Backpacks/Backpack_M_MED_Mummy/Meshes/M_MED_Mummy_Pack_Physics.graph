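# Physics asset for the Mummy medium male backpack cosmetic, providing physics bodies for the pack and its dangling strap.
# Single capsules cover strap_MSTR, two dyn_strap bones, the pack and spine_01.
# Two constraints chain strap_MSTR through dyn_strap_01 to dyn_strap_02, with identical swing angles and a narrow twist angle.
# The pack and spine capsules have no constraints, apparently serving as collision volumes for the strap.
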PhysicsAsset /BRCosmetics/Accessories/FORT_Backpacks/Backpack_M_MED_Mummy/Meshes/M_MED_Mummy_Pack_Physics
bodies {
 dyn_strap_01 [{shapeType=Capsule, length=3, radius=2, shapeName=dyn_strap_02_R_capsule, center=<1.2611 0 0>, rotation=<rot 0 -179.9995 -180>}]
 dyn_strap_02 [{shapeType=Capsule, length=3, radius=2, shapeName=dyn_strap_02_R_capsule, center=<1.2611 0 0>, rotation=<rot 0 0.0003 0>}]
 pack [{shapeType=Capsule, length=20, radius=6, shapeName=pack_capsule, center=<-1.2895 -5.4209 -10.9194>, rotation=<rot 0 -90 0>}]
 spine_01 [{shapeType=Capsule, length=15, radius=10.5, shapeName=spine_01_capsule}]
 strap_MSTR [{shapeType=Capsule, length=0, radius=0.5, shapeName=strap_MSTR_L_capsule, rotation=<rot 90 -90.0001 -90.0005>}]
}
constraints [
 {bone1Name=strap_MSTR, bone2Name=dyn_strap_01, swing1LimitDegrees=45, swing2LimitDegrees=45, twistLimitDegrees=5}
 {bone1Name=dyn_strap_02, bone2Name=dyn_strap_01, swing1LimitDegrees=45, swing2LimitDegrees=45, twistLimitDegrees=5}
]
