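# Animation Blueprint for the Mummy backpack cosmetic, adding secondary motion to its straps and zippers.
# The EventGraph copies the body mesh's spine_01 world transform onto the pack and updates a gravity override.
# The AnimGraph applies a skydiving root offset and zipper AnimDynamics, then caches that pose.
# From that cache, bool blends select RigidBody physics, strap AnimDynamics or a cheaper Trail controller before a final pack offset.
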AnimBlueprint /BRCosmetics/Accessories/FORT_Backpacks/Backpack_M_MED_Mummy/Meshes/M_MED_Mummy_Pack_AnimBP
target_skeleton=&/M_MED_Mummy_Pack_Skeleton
graphs [
 AnimGraph {nodes [
   n1 = Anim Dynamics - Bone: dyn_strap_01!(
    ComponentPose=Local To Component(LocalPose=Use cached pose 'pre rigid body pose'())
    Alpha=1
    GravityOverride:Vector="0,0,-980"
    ExternalForce="0,0,0"
   ) -> (Pose) {
    angularBiasOverride=1
    angularDampingOverride=1
    angularSpringConstant=3000
    bAngularSpring=true
    bOverrideAngularBias=true
    bOverrideAngularDamping=true
    bUseGravityOverride=true
    simulationSpace=RootRelative
    alphaBoolBlend {blendOption=Linear}
    boundBone {boneName=dyn_strap_01}
    componentAppliedLinearAccClamp=<3000 3000 3000>
    componentLinearAccScale=<0.4 0.5 0.5>
    componentLinearVelScale=<0.5 0.5 0.5>
    physicsBodyDefinitions [{
      collisionType=CustomSphere
      sphereCollisionRadius=2.2
      boundBone {boneName=dyn_strap_01}
      boxExtents=<18 14 14>
      constraintSetup {angularLimitsMax=<5 35 0>, angularLimitsMin=<-3 -30 -130>, angularTarget=<0 0 0>}
      localJointOffset=<5 0 0>
     }]
    planarLimits [{drivingBone {boneName=pack}, planeTransform {location=<0.354 -0.2986 0>, rotation=<rot 0.0001 10 -90.0003>}}]
   }
   n2 = Anim Dynamics - Bone: dyn_strap_02!(ComponentPose=@n1, Alpha=1, GravityOverride:Vector="0,0,-980", ExternalForce="0,0,0") -> (Pose) {
    angularBiasOverride=1
    angularDampingOverride=1
    angularSpringConstant=3000
    bAngularSpring=true
    bOverrideAngularBias=true
    bOverrideAngularDamping=true
    bUseGravityOverride=true
    simulationSpace=RootRelative
    alphaBoolBlend {blendOption=Linear}
    boundBone {boneName=dyn_strap_02}
    componentAppliedLinearAccClamp=<3000 3000 3000>
    componentLinearAccScale=<0.4 0.5 0.5>
    componentLinearVelScale=<0.5 0.5 0.5>
    physicsBodyDefinitions [{
      collisionType=CustomSphere
      sphereCollisionRadius=3
      boundBone {boneName=dyn_strap_02}
      boxExtents=<18 12 12>
      constraintSetup {angularLimitsMax=<5 45 45>, angularLimitsMin=<-5 -45 -80>, angularTarget=<0 0 0>}
      localJointOffset=<6 0 0>
     }]
    planarLimits [{drivingBone {boneName=pack}, planeTransform {location=<0.152 0.847 0>, rotation=<rot 0.0001 10 -90.0003>}}]
   }
   n3 = Trail controller - Bone: dyn_strap_02(ComponentPose=Local To Component(LocalPose=Use cached pose 'pre rigid body pose'()), Alpha=1) {
    bLimitRotation=true
    bReorientParentToChild=false
    bShowTrailLocation=true
    maxDeltaTime=0.333
    alphaScaleBiasClamp {bInterpResult=true}
    fakeVelocity=<1 1 300>
    planarLimits [{drivingBone {boneName=pack}, planeTransform {location=<0 0 0>, rotation=<rot 0 0 0>, scale=<1 1 1>}}]
    rotationLimits [{limitMax=<9 9 9>, limitMin=<-9 -9 -9>}, {limitMax=<15 15 15>, limitMin=<-15 -15 -15>}]
    rotationOffsets [<0 0 0>, <0 0 0>]
    trailBone {boneName=dyn_strap_02}
    trailRelaxationSpeed {editorCurveData {keys [<at -0.01 10 RCIM_Linear>,<at 0.2519 11.3498 RCIM_Linear>,<at 0.9955 9.604 RCIM_Linear>]}}
   }
   n4 = Blend Poses by bool(
    BlendPose_0=Component To Local(ComponentPose=@n2)
    BlendPose_1=Component To Local(ComponentPose=@n3)
    bActiveValue=Get bAnimDynamicsEnabled() -> (bAnimDynamicsEnabled:bool)
   ) {blendType=Linear}
   n5 = RigidBody!(
    ComponentPose=Local To Component(LocalPose=Use cached pose 'pre rigid body pose'())
    Alpha
    OverrideWorldGravity:Vector=Get Gravity Override()
    ExternalForce
   ) -> (Pose) {
    bClampLinearTranslationLimitToRefPose=true
    bOverrideWorldGravity=true
    overridePhysicsAsset=&/M_MED_Mummy_Pack_Physics
    simulationSpace=BaseBoneSpace
    alphaBoolBlend {blendOption=Linear}
    baseBoneRef {boneName=pack}
    componentAppliedLinearAccClamp=<3000 3000 3000>
    componentLinearAccScale=<0.4 0.4 0.5>
    componentLinearVelScale=<0.4 0.4 0.5>
    overrideWorldGravity {z=-980}
   }
   n6 = Component To Local(ComponentPose=@n5)
   n7 = Get bRigidBodyEnabled()
   n8 = "Transform (Modify) Bone - Bone: pack"!(
    ComponentPose=Local To Component(LocalPose=Blend Poses by bool(BlendPose_0=@n6, BlendPose_1=@n4, bActiveValue=@n7) {blendType=Linear})
    Alpha
    Translation="0,0,5"
   ) -> (Pose) {
    translationMode=BMM_Additive
    translationSpace=BCS_BoneSpace
    alphaBoolBlend {blendInTime=1, blendOption=Linear, blendOutTime=1}
    boneToModify {boneName=pack}
    translation {z=5}
   }
   Output Pose(Result=Component To Local(ComponentPose=@n8)) {layerGroup=None}
   n9 = "Transform (Modify) Bone - Bone: root"!(
    ComponentPose=Local To Component(LocalPose=Input Pose())
    bAlphaBoolEnabled:bool=Get bIsSkyDiving()
    Translation="0,0.5,3.5"
    Rotation="0,0,0"
    Scale
   ) -> (Pose) {
    alphaInputType=Bool
    translationMode=BMM_Additive
    translationSpace=BCS_BoneSpace
    alphaBoolBlend {blendInTime=0.2, blendOption=Linear, blendOutTime=0.2}
    boneToModify {boneName=root}
    translation {y=0.5, z=3.5}
   }
   n10 = "Transform (Modify) Bone - Bone: spine_01"(
    ComponentPose=@n9
    Alpha=1
    Translation=Get spine_01 translation() -> (spine_01 translation:Vector)
    Rotation=Get spine_01rotation() -> (spine_01rotation:Rotator)
    Scale="1,1,1"
   ) {
    rotationMode=BMM_Replace
    rotationSpace=BCS_WorldSpace
    scaleMode=BMM_Additive
    translationMode=BMM_Replace
    translationSpace=BCS_WorldSpace
    alphaBoolBlend {blendInTime=1, blendOption=Linear, blendOutTime=1}
    boneToModify {boneName=spine_01}
    rotation {roll=20}
   }
   n11 = Anim Dynamics - Bone: dyn_zipper_A_right(ComponentPose=@n10, Alpha=1, ExternalForce="0,0,-980") {
    angularDampingOverride=0.999
    angularSpringConstant=500
    bAngularSpring=true
    bOverrideAngularDamping=true
    bUseGravityOverride=true
    gravityScale=0
    simulationSpace=RootRelative
    alphaBoolBlend {blendOption=Linear}
    boundBone {boneName=dyn_zipper_A_right}
    componentAppliedLinearAccClamp=<3000 3000 3000>
    componentLinearAccScale=<0.45 0.4 0.45>
    componentLinearVelScale=<0.7 0.6 0.7>
    gravityOverride {z=-980}
    physicsBodyDefinitions [{
      boundBone {boneName=dyn_zipper_A_right}
      boxExtents=<5.5 6.5 3>
      constraintSetup {angularLimitsMax=<5 35 0>, angularLimitsMin=<-3 -30 -130>, angularTarget=<1 0 0>}
      localJointOffset=<1 0 0>
     }]
    relativeSpaceBone {boneName=root}
   }
   n12 = Anim Dynamics - Bone: dyn_zipper_A_left(ComponentPose=@n11, Alpha=1, ExternalForce="0,0,-980") {
    angularDampingOverride=0.999
    angularSpringConstant=500
    bAngularSpring=true
    bOverrideAngularDamping=true
    bUseGravityOverride=true
    gravityScale=0
    simulationSpace=RootRelative
    alphaBoolBlend {blendOption=Linear}
    boundBone {boneName=dyn_zipper_A_left}
    componentAppliedLinearAccClamp=<3000 3000 3000>
    componentLinearAccScale=<0.4 0.45 0.4>
    componentLinearVelScale=<0.6 0.7 0.6>
    gravityOverride {z=-980}
    physicsBodyDefinitions [{
      boundBone {boneName=dyn_zipper_A_left}
      boxExtents=<6.5 5.5 3>
      constraintSetup {angularLimitsMax=<3 30 0>, angularLimitsMin=<-5 -35 -135>, angularTarget=<1 0 0>}
      localJointOffset=<1 0 0>
     }]
    relativeSpaceBone {boneName=root}
   }
   Save cached pose 'pre rigid body pose'(Pose=Component To Local(ComponentPose=@n12)) {cacheName=pre rigid body pose}
  ]}
 EventGraph {nodes [
   Event BlueprintUpdateAnimation()
   > Is Valid(InputObject=Get BodyMesh() -> (BodyMesh:Skeletal Mesh Component Object Reference))
   n1 = BreakTransform(InTransform=GetSocketTransform(self=Get BodyMesh(), InSocketName=spine_01, TransformSpace=RTS_World))
   > Set spine_01 translation(spine_01 translation:Vector=@n1) -> (Output_Get:Vector)
   > Set spine_01rotation(spine_01rotation:Rotator=@n1.Rotation) -> (Output_Get:Rotator)
   > n2 = Gravity Override State Function DEPRECATED(CharacterPartAnimInstance=Self-Reference(), GameTimeInSeconds=GetGameTimeInSeconds(), GravityOverrideParams=Get Gravity Override Params())
   > Set Gravity Override(Gravity Override=@n2)
   n3 = Event OnCharacterCustomizationLoaded()
   > Set BodyMesh(BodyMesh=Get Array Item with EFort Custom Part Type(Array=@n3.CustomMeshParts))
  ]}
]
variables [
 BodyMesh {type=SkeletalMeshComponent}
 Gravity Override {type=Vector}
 Gravity Override Params {type=GravityOverrideParamsStruct}
 spine_01 translation {type=Vector}
 spine_01rotation {type=Rotator}
]
